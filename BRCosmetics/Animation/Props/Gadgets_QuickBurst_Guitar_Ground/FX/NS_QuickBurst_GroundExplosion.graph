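# Niagara ground explosion effect for the QuickBurst Guitar gadget prop cosmetic, throwing out dirt, dust and streaks in one-shot bursts.
# Dirt is a GPU emitter with collision, and Dirt_Low repeats it without GPU simulation or collision, both using spawn groups for large and small debris.
# Dust layers immediate and delayed puffs with curve-driven material parameters and time-of-day scaling.
# Streaks fires slightly delayed, stretching sprites; Streaks_LWE has no renderers.
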NiagaraSystem /BRCosmetics/Animation/Props/Gadgets_QuickBurst_Guitar_Ground/FX/NS_QuickBurst_GroundExplosion
dependencies {
 usedDataInterfaces [
  &NiagaraDataInterfaceCurve
  &NiagaraDataInterfaceVector2DCurve
  &NiagaraDataInterfaceColorCurve
  &NiagaraDataInterfaceSpriteRendererInfo
  &NiagaraDataInterfaceCollisionQuery
  &NiagaraDataInterfaceLandscape
 ]
 usedDynamicInputs []
 usedModules [&SystemState]
}
emitters {
 Dirt {
  simTarget=GPUComputeSim
  emitterUpdateScript [
   &EmitterState()
   &SpawnBurst_Instantaneous(
    Spawn Count=&RandomRangeInt(Minimum=5, Maximum=25)
    Use Spawn Probability=false
    UseLoopCountLimit=false
   )
   &SpawnBurst_Instantaneous(
    Spawn Count=&RandomRangeInt(Minimum=100, Maximum=250)
    Spawn Group=1
    Use Spawn Probability=false
    UseLoopCountLimit=false
   ) {module=SpawnBurst_Instantaneous001}
  ]
  particleSpawnScript [
   /SetVariables(Particles.VisibilityTag=&SpawnGroup()) {isSetParameters=true, module=SetVariables_#1}
   &InitializeParticle(
    Color=&MaskLinearColorBySpawnGroup(
     Masked Linear Color=&RandomRangeLinearColor(Minimum=<rgba 0.2552 0.1591 0.0828 1>, Maximum=<rgba 0.3542 0.253 0.1952 1>)
     Passthrough Linear Color=&LerpLinearColors(
      StartColor=&RandomRangeLinearColor(Minimum=<rgba 0.0729 0.0681 0.063 1>, Maximum=<rgba 0.1458 0.1349 0.1253 1>)
      EndColor=&RandomRangeLinearColor(Minimum=<rgba 0.2552 0.1591 0.0828 1>, Maximum=<rgba 0.3542 0.253 0.1952 1>)
      LerpFactor=&RandomRangeFloat(Minimum=0, Maximum=1)
     )
    )
    Lifetime=&MaskFloatBySpawnGroup(
     Masked Float=&RandomRangeFloat(Minimum=0.5, Maximum=1.5)
     Passthrough Float=&RandomRangeFloat(Minimum=0.675, Maximum=1.75)
    )
    Sprite Rotation Mode=::NewEnumerator2
    Sprite Size Mode=::NewEnumerator1
    Uniform Sprite Size Max=&MaskFloatBySpawnGroup(Masked Float=95, Passthrough Float=8)
    Uniform Sprite Size Min=&MaskFloatBySpawnGroup(Masked Float=45, Passthrough Float=2)
    UsePositionOffset=false
    Write Material Random=true
   )
   &ShapeLocation(
    Cylinder Height=1
    Cylinder Radius=50
    MaskSpawnGroup=true
    Override Seed=false
    Shape Primitive=::NewEnumerator1
    Surface Only=true
    Use Endcaps In Surface Only Mode=false
    UseOwnerScale=true
   )
   &ShapeLocation(
    Cylinder Height=1
    Cylinder Radius=65
    MaskSpawnGroup=true
    Override Seed=false
    Shape Primitive=::NewEnumerator1
    Spawn Group Mask=1
    Surface Only=true
    Surface Only Band Thickness=0.25
    Use Endcaps In Surface Only Mode=false
    UseOwnerScale=true
   ) {module=ShapeLocation001}
   &AddVelocity(
    Module.Use Falloff Distance=false
    Origin Offset=<0 0 -45>
    Override Seed=false
    Velocity Mode=::NewEnumerator1
    Velocity Speed=&MaskFloatBySpawnGroup(
     Masked Float=&RandomRangeFloat(Minimum=150, Maximum=650)
     Passthrough Float=&RandomRangeFloat(Minimum=225, Maximum=750)
    )
   )
  ]
  particleUpdateScript [
   &ParticleState()
   &GravityForce(Gravity=&Multiply_VectorByFloat(Vector=<0 0 -980>, Float=&RandomRangeFloat(Minimum=0.5, Maximum=1.5)))
   &WindForce(
    Scale=2
    Turbulence Offset Mode:ENiagaraWindOffsetMode=::NewEnumerator1
    Wind Speed=&RandomRangeVector(Minimum=<0 0 0>, Maximum=<1 1 1>)
    Wind Speed Scale=&RandomRangeFloat(Minimum=100, Maximum=200)
   )
   &Drag(
    Drag=&MaskFloatBySpawnGroup(
     Masked Float=&RandomRangeFloat(Minimum=0.5, Maximum=1.5)
     Passthrough Float=&RandomRangeFloat(Minimum=1, Maximum=3)
    )
    Use Linear Drag=true
    Use Rotational Drag=false
   )
   &Collision(
    Age Colliding Particles=false
    Collision Enabled=&SetBoolByFloatComparison(A=$Particles.Age, B=0.5)
    GPU Collision Type=::NewEnumerator2
    KillOnCollision=false
    Module.Control Rest State Via Surface Normal=false
    Module.EnableMaxCollisionCount=false
    Module.Randomize Collision Normal Vector=false
    Restitution=0.75
   )
   &SolveForcesAndVelocity(Clamp Velocity=false, Limit Acceleration=false)
   &ScaleColor(
    Scale Alpha=&FloatFromCurve(FloatCurve=DataInterface() {propertyValues {Curve {keys [{interpMode=RCIM_Cubic, tangentMode=RCTM_User, time=0, value=1}, {interpMode=RCIM_Cubic, tangentMode=RCTM_User, time=1, value=0}]}}})
    Scale RGB=&VectorFromFloat(Value=&MaskFloatBySpawnGroup(Masked Float=&RandomRangeFloat(Minimum=0.1, Maximum=1), Passthrough Float=1))
    ScaleA=true
    ScaleRGB=true
   )
   &ScaleSpriteSize(
    Scale Sprite Size Mode=::NewEnumerator0
    Uniform Scale Factor=&MaskFloatBySpawnGroup(
     Masked Float=&FloatFromCurve(
      FloatCurve=DataInterface() {propertyValues {Curve {keys [
          {arriveTangent=9.2667, interpMode=RCIM_Cubic, leaveTangent=9.2667, tangentMode=RCTM_User, time=0, value=0.2}
          {arriveTangentWeight=0.1414, interpMode=RCIM_Cubic, leaveTangentWeight=0.3299, tangentMode=RCTM_User, time=0.2, value=1}
          {arriveTangent=-1.3147, interpMode=RCIM_Cubic, leaveTangent=-1.3147, tangentMode=RCTM_User, time=1, value=0.3}
         ]}}}
      Scale Curve=0.85
     )
     Passthrough Float=&FloatFromCurve(FloatCurve=DataInterface() {propertyValues {Curve {keys [<at 0.8 1 RCIM_Cubic>,<at 1 0 RCIM_Cubic>]}}})
    )
   )
   &SpriteRotationRate(Rotation Rate=&Multiply_Float(
     A=&RandomRangeFloat(Minimum=0.1, Maximum=0.9)
     B=&MaskFloatBySpawnGroup(
      Masked Float=&RandomRangeFloat(Minimum=-90, Maximum=90)
      Passthrough Float=&Multiply_Float(
       A=&RandomRangeFloat(Minimum=-90, Maximum=90)
       B=&MakeCustomFloatFromBool(Bool=$Particles.HasCollided, True Float=0, False Float=1)
      )
     )
    ))
  ]
  rendererClasses [&NiagaraSpriteRendererProperties]
  renderers [
   {rendererClass=&NiagaraSpriteRendererProperties}
   {rendererClass=&NiagaraSpriteRendererProperties, rendererIndex=1}
  ]
 }
 Dirt_Low {
  emitterUpdateScript [
   &EmitterState()
   &SpawnBurst_Instantaneous(
    Spawn Count=&RandomRangeInt(Minimum=10, Maximum=20)
    Use Spawn Probability=false
    UseLoopCountLimit=false
   )
   &SpawnBurst_Instantaneous(
    Spawn Count=&RandomRangeInt(Minimum=50, Maximum=200)
    Spawn Group=1
    Use Spawn Probability=false
    UseLoopCountLimit=false
   ) {module=SpawnBurst_Instantaneous001}
  ]
  particleSpawnScript [
   /SetVariables(Particles.VisibilityTag=&SpawnGroup()) {isSetParameters=true, module=SetVariables_#1}
   &InitializeParticle(
    Color=&MaskLinearColorBySpawnGroup(
     Masked Linear Color=&RandomRangeLinearColor(Minimum=<rgba 0.2552 0.1591 0.0828 1>, Maximum=<rgba 0.3542 0.253 0.1952 1>)
     Passthrough Linear Color=&LerpLinearColors(
      StartColor=&RandomRangeLinearColor(Minimum=<rgba 0.0729 0.0681 0.063 1>, Maximum=<rgba 0.1458 0.1349 0.1253 1>)
      EndColor=&RandomRangeLinearColor(Minimum=<rgba 0.2552 0.1591 0.0828 1>, Maximum=<rgba 0.3542 0.253 0.1952 1>)
      LerpFactor=&RandomRangeFloat(Minimum=0, Maximum=1)
     )
    )
    Lifetime=&MaskFloatBySpawnGroup(
     Masked Float=&RandomRangeFloat(Minimum=0.5, Maximum=1.5)
     Passthrough Float=&RandomRangeFloat(Minimum=0.375, Maximum=1.375)
    )
    Sprite Rotation Mode=::NewEnumerator2
    Sprite Size Mode=::NewEnumerator1
    Uniform Sprite Size Max=&MaskFloatBySpawnGroup(Masked Float=95, Passthrough Float=8)
    Uniform Sprite Size Min=&MaskFloatBySpawnGroup(Masked Float=45, Passthrough Float=2)
    UsePositionOffset=false
    Write Material Random=true
   )
   &ShapeLocation(
    Cylinder Height=1
    Cylinder Radius=50
    MaskSpawnGroup=true
    Override Seed=false
    Shape Primitive=::NewEnumerator1
    Surface Only=true
    Use Endcaps In Surface Only Mode=false
    UseOwnerScale=true
   )
   &ShapeLocation(
    Cylinder Height=1
    Cylinder Radius=65
    MaskSpawnGroup=true
    Override Seed=false
    Shape Primitive=::NewEnumerator1
    Spawn Group Mask=1
    Surface Only=true
    Surface Only Band Thickness=0.25
    Use Endcaps In Surface Only Mode=false
    UseOwnerScale=true
   ) {module=ShapeLocation001}
   &AddVelocity(
    Module.Use Falloff Distance=false
    Origin Offset=<0 0 -45>
    Override Seed=false
    Velocity Mode=::NewEnumerator1
    Velocity Speed=&MaskFloatBySpawnGroup(
     Masked Float=&RandomRangeFloat(Minimum=150, Maximum=650)
     Passthrough Float=&RandomRangeFloat(Minimum=225, Maximum=750)
    )
   )
  ]
  particleUpdateScript [
   &ParticleState()
   &GravityForce(Gravity=&Multiply_VectorByFloat(Vector=<0 0 -980>, Float=&RandomRangeFloat(Minimum=0.5, Maximum=1.5)))
   &WindForce(
    Scale=2
    Turbulence Offset Mode:ENiagaraWindOffsetMode=::NewEnumerator1
    Wind Speed=&RandomRangeVector(Minimum=<0 0 0>, Maximum=<1 1 1>)
    Wind Speed Scale=&RandomRangeFloat(Minimum=100, Maximum=200)
   )
   &Drag(
    Drag=&MaskFloatBySpawnGroup(
     Masked Float=&RandomRangeFloat(Minimum=0.5, Maximum=1.5)
     Passthrough Float=&RandomRangeFloat(Minimum=1, Maximum=3)
    )
    Use Linear Drag=true
    Use Rotational Drag=false
   )
   &SolveForcesAndVelocity(Clamp Velocity=false, Limit Acceleration=false)
   &ScaleColor(
    Scale Alpha=&FloatFromCurve(FloatCurve=DataInterface() {propertyValues {Curve {keys [{interpMode=RCIM_Cubic, tangentMode=RCTM_User, time=0, value=1}, {interpMode=RCIM_Cubic, tangentMode=RCTM_User, time=1, value=0}]}}})
    Scale RGB=&VectorFromFloat(Value=&MaskFloatBySpawnGroup(Masked Float=&RandomRangeFloat(Minimum=0.1, Maximum=1), Passthrough Float=1))
    ScaleA=true
    ScaleRGB=true
   )
   &ScaleSpriteSize(
    Scale Sprite Size Mode=::NewEnumerator0
    Uniform Scale Factor=&MaskFloatBySpawnGroup(
     Masked Float=&FloatFromCurve(
      FloatCurve=DataInterface() {propertyValues {Curve {keys [
          {arriveTangent=9.2667, interpMode=RCIM_Cubic, leaveTangent=9.2667, tangentMode=RCTM_User, time=0, value=0.2}
          {arriveTangentWeight=0.1414, interpMode=RCIM_Cubic, leaveTangentWeight=0.3299, tangentMode=RCTM_User, time=0.2, value=1}
          {arriveTangent=-1.3147, interpMode=RCIM_Cubic, leaveTangent=-1.3147, tangentMode=RCTM_User, time=1, value=0.3}
         ]}}}
      Scale Curve=0.85
     )
     Passthrough Float=1
    )
   )
   &SpriteRotationRate(Rotation Rate=&Multiply_Float(
     A=&RandomRangeFloat(Minimum=0.1, Maximum=0.9)
     B=&RandomRangeFloat(Minimum=-90, Maximum=90)
    ))
  ]
  rendererClasses [&NiagaraSpriteRendererProperties]
  renderers [
   {rendererClass=&NiagaraSpriteRendererProperties}
   {rendererClass=&NiagaraSpriteRendererProperties, rendererIndex=1}
  ]
 }
 Dust {
  emitterUpdateScript [
   &EmitterState()
   &SpawnBurst_Instantaneous(
    Spawn Count=&RandomRangeInt(Minimum=8, Maximum=12)
    Use Spawn Probability=false
    UseLoopCountLimit=false
   )
   &SpawnBurst_Instantaneous(
    Spawn Count=&RandomRangeInt(Minimum=8, Maximum=12)
    Spawn Group=2
    Use Spawn Probability=false
    UseLoopCountLimit=false
   ) {module=SpawnBurst_Instantaneous002}
   &SpawnBurst_Instantaneous(
    Spawn Count=&RandomRangeInt(Minimum=8, Maximum=12)
    Spawn Group=1
    Spawn Time=0.1
    Use Spawn Probability=false
    UseLoopCountLimit=false
   ) {module=SpawnBurst_Instantaneous001}
   &SpawnBurst_Instantaneous(
    Spawn Count=&RandomRangeInt(Minimum=8, Maximum=12)
    Spawn Group=3
    Spawn Time=0.1
    Use Spawn Probability=false
    UseLoopCountLimit=false
   ) {module=SpawnBurst_Instantaneous003}
  ]
  particleSpawnScript [
   /SetVariables(Particles.VisibilityTag=&MakeCustomIntFromBool(
     Bool=&SetBoolByIntComparison(A=&SpawnGroup(), B=2, Comparison Type=::NewEnumerator2)
     True INT=1
     False INT=0
    )) {isSetParameters=true}
   /SetVariables(
    Particles.SizeMax:NiagaraFloat=&MakeCustomFloatFromBool(
     Bool=&SetBoolByIntComparison(A=&SpawnGroup(), B=2, Comparison Type=::NewEnumerator2)
     True Float=&RandomRangeFloat(Minimum=45, Maximum=65)
     False Float=&RandomRangeFloat(Minimum=155, Maximum=185)
    )
    Particles.SizeMin:NiagaraFloat=&MakeCustomFloatFromBool(
     Bool=&SetBoolByIntComparison(A=&SpawnGroup(), B=2, Comparison Type=::NewEnumerator2)
     True Float=&RandomRangeFloat(Minimum=65, Maximum=85)
     False Float=&RandomRangeFloat(Minimum=165, Maximum=185)
    )
   ) {isSetParameters=true}
   /SetVariables(Particles.SizeCurve:NiagaraFloat=&Lerp_Float(
     A=$Particles.SizeMin
     B=$Particles.SizeMax
     Alpha=&FloatFromCurve(
      FloatCurve=DataInterface() {propertyValues {Curve {keys [
          {interpMode=RCIM_Cubic, tangentMode=RCTM_Break, time=0, value=0}
          {arriveTangentWeight=0.9217, interpMode=RCIM_Cubic, tangentMode=RCTM_Break, time=1, value=1}
         ]}}}
      CurveIndex=&RandomRangeFloat(Minimum=0, Maximum=1)
     )
    )) {isSetParameters=true}
   &InitializeParticle(
    Color=&MakeLinearColorFromVectorAndFloat(
     "Vector (RGB)"=<1 1 1>
     "Float (Alpha)"=&MaskFloatBySpawnGroup(
      Masked Float=&RandomRangeFloat(Minimum=0.125, Maximum=0.25)
      Passthrough Float=&MaskFloatBySpawnGroup(
       Masked Float=&RandomRangeFloat(Minimum=0.25, Maximum=1)
       Passthrough Float=&RandomRangeFloat(Minimum=0.1, Maximum=0.3)
       Spawn Group Mask=1
      )
     )
    )
    Lifetime=&MaskFloatBySpawnGroup(
     Masked Float=&RandomRangeFloat(Minimum=1, Maximum=2)
     Passthrough Float=&MaskFloatBySpawnGroup(Masked Float=&RandomRangeFloat(Minimum=0.5, Maximum=1.25), Passthrough Float=1, Spawn Group Mask=1)
    )
    Sprite Rotation Mode=::NewEnumerator2
    Sprite Size Mode=::NewEnumerator3
    Uniform Sprite Size=$Particles.SizeCurve
    UsePositionOffset=false
    Write Material Random=true
   )
   &ShapeLocation(Hemisphere Distribution=<1 0.5>, MaskSpawnGroup=true, Override Seed=false, Spawn Group Mask=1, Sphere Radius=50, UseOwnerScale=true)
   &ShapeLocation(Hemisphere Distribution=<1 0.5>, MaskSpawnGroup=true, Override Seed=false, Spawn Group Mask=3, Sphere Radius=50, UseOwnerScale=true) {module=ShapeLocation003}
   &ShapeLocation(
    Cylinder Height=1
    Cylinder Height Midpoint=0
    MaskSpawnGroup=true
    Override Seed=false
    Shape Primitive=::NewEnumerator1
    Surface Only=true
    Use Endcaps In Surface Only Mode=false
    UseOwnerScale=true
   ) {module=ShapeLocation001}
   &ShapeLocation(
    Cylinder Height=1
    Cylinder Height Midpoint=0
    MaskSpawnGroup=true
    Override Seed=false
    Shape Primitive=::NewEnumerator1
    Spawn Group Mask=2
    Surface Only=true
    Use Endcaps In Surface Only Mode=false
    UseOwnerScale=true
   ) {module=ShapeLocation002}
   &AddVelocity(
    Velocity=&MaskVectorBySpawnGroup(Masked Vector=$Particles.ShapeLocation.ShapeNormal, Passthrough Vector=$Particles.ShapeLocation003.ShapeNormal, Spawn Group Mask=1)
    Velocity Speed Scale=&MaskFloatBySpawnGroup(Masked Float=0, Passthrough Float=&RandomRangeFloat(Minimum=200, Maximum=500))
   )
   &AddVelocity(
    Module.Use Falloff Distance=false
    Origin Offset=<0 0 -50>
    Override Seed=false
    Velocity Mode=::NewEnumerator1
    Velocity Speed=&MaskFloatBySpawnGroup(
     Masked Float=75
     Passthrough Float=&MaskFloatBySpawnGroup(Masked Float=75, Passthrough Float=0, Spawn Group Mask=2)
    )
   ) {module=AddVelocity001}
   &CameraOffset(Camera Offset Amount=75)
   &SubUVAnimation(
    Evaluation Type=::NewEnumerator0
    "Sprite Renderer "=DataInterface() {propertyValues {SpriteRenderer=&self:Dust_2.NiagaraSpriteRendererProperties_2}}
    SubUV Animation Mode:ENiagara_SubUVLookupModeV2=::NewEnumerator2
    UseEndFrame=false
    UseStartFrame=false
   )
  ]
  particleUpdateScript [
   &ParticleState()
   &ScaleSpriteSize(
    Scale Sprite Size Mode=::NewEnumerator0
    Uniform Scale Factor=&MaskFloatBySpawnGroup(
     Masked Float=&FloatFromCurve(FloatCurve=DataInterface() {propertyValues {Curve {keys [{arriveTangent=6.1888, interpMode=RCIM_Cubic, leaveTangent=6.1888, tangentMode=RCTM_User, time=0, value=0}, <at 0.26 1 RCIM_Cubic>]}}})
     Passthrough Float=1
    )
   )
   &ScaleColor(
    Scale Alpha=&FloatFromCurve(
     FloatCurve=DataInterface() {propertyValues {Curve {keys [<at 0 1>,<at 1 0>]}}}
     Scale Curve=&RandomRangeFloat(Minimum=0.5, Maximum=0.7)
    )
    Scale RGB=&MakeVectorFromLinearColorRGB(Color=&LerpLinearColors(
      StartColor=&RandomRangeLinearColor(
       Minimum=&ColorFromCurve(ColorCurve=DataInterface() {propertyValues {
          AlphaCurve {keys []}
          BlueCurve {keys [<at 0 0.063 RCIM_Cubic>,<at 1 0.1495 RCIM_Cubic>]}
          GreenCurve {keys [<at 0 0.0681 RCIM_Cubic>,<at 1 0.1616 RCIM_Cubic>]}
          RedCurve {keys [<at 0 0.0729 RCIM_Cubic>,<at 1 0.1729 RCIM_Cubic>]}
         }})
       Maximum=&ColorFromCurve(ColorCurve=DataInterface() {propertyValues {
          AlphaCurve {keys []}
          BlueCurve {keys [<at 0 0.1253 RCIM_Cubic>,<at 1 0.2113 RCIM_Cubic>]}
          GreenCurve {keys [<at 0 0.1349 RCIM_Cubic>,<at 1 0.2274 RCIM_Cubic>]}
          RedCurve {keys [<at 0 0.1458 RCIM_Cubic>,<at 1 0.2458 RCIM_Cubic>]}
         }})
      )
      EndColor=&RandomRangeLinearColor(
       Minimum=&ColorFromCurve(ColorCurve=DataInterface() {propertyValues {
          AlphaCurve {keys []}
          BlueCurve {keys [<at 0 0.0828 RCIM_Cubic>,<at 1 0.1153 RCIM_Cubic>]}
          GreenCurve {keys [<at 0 0.1591 RCIM_Cubic>,<at 1 0.2215 RCIM_Cubic>]}
          RedCurve {keys [<at 0 0.2552 RCIM_Cubic>,<at 1 0.3552 RCIM_Cubic>]}
         }})
       Maximum=&ColorFromCurve(ColorCurve=DataInterface() {propertyValues {
          AlphaCurve {keys []}
          BlueCurve {keys [<at 0 0.1952 RCIM_Cubic>,<at 1 0.2504 RCIM_Cubic>]}
          GreenCurve {keys [<at 0 0.253 RCIM_Cubic>,<at 1 0.3244 RCIM_Cubic>]}
          RedCurve {keys [<at 0 0.3542 RCIM_Cubic>,<at 1 0.4542 RCIM_Cubic>]}
         }})
      )
      LerpFactor=&RandomRangeFloat(Minimum=0, Maximum=1)
     ))
    ScaleA=true
    ScaleRGB=true
   )
   &GravityForce(Gravity=&RandomRangeVector(Minimum=<0 0 -150>, Maximum=<0 0 -300>))
   &Drag(
    Drag=&FloatFromCurve(
     FloatCurve=DataInterface() {propertyValues {Curve {keys [{interpMode=RCIM_Cubic, tangentMode=RCTM_Break, time=0, value=1}, {interpMode=RCIM_Cubic, tangentMode=RCTM_Break, time=0.8, value=0}]}}}
     Scale Curve=&MaskFloatBySpawnGroup(Masked Float=1, Passthrough Float=4)
    )
    Use Linear Drag=true
    Use Rotational Drag=false
   )
   &PointAttractionForce(
    Attraction Radius=150
    Attraction Strength=&MaskFloatBySpawnGroup(Masked Float=-10, Passthrough Float=0)
    Kill Within Radius=false
    Module.Use Falloff=true
    UseAttractorOffset=false
   )
   &SolveForcesAndVelocity(Clamp Velocity=false, Limit Acceleration=false)
   &DynamicMaterialParameters(
    Index 0 Param 1=&Add_Float(
     A=&RandomRangeFloat(Minimum=0, Maximum=1)
     B=&FloatFromCurve(
      FloatCurve=DataInterface() {propertyValues {Curve {keys [
          {arriveTangent=2.7533, interpMode=RCIM_Cubic, leaveTangent=1.8958, tangentMode=RCTM_User, time=0, value=1}
          {arriveTangent=0.8292, interpMode=RCIM_Cubic, leaveTangent=0.8292, tangentMode=RCTM_User, time=0.321, value=1.3453}
          {arriveTangent=0.0529, interpMode=RCIM_Cubic, leaveTangent=0.2021, tangentMode=RCTM_User, time=1, value=1.5}
         ]}}}
      CurveIndex=&Add_Float(A=&Multiply_Float(A=$Particles.Initial.MaterialRandom, B=0.5), B=$Particles.NormalizedAge)
      Scale Curve=&Lerp_Float(
       A=0.25
       B=0.7
       Alpha=&NormalizeFloat(Value=$Particles.SizeCurve, Min=$Particles.SizeMin, Max=$Particles.SizeMax)
      )
     )
    )
    Index 0 Param 2=&FloatFromCurve(FloatCurve=DataInterface() {propertyValues {Curve {keys [<at 0 1>,<at 1 0>]}}})
    Index 0 Param 3=&FloatFromCurve(
     FloatCurve=DataInterface() {propertyValues {Curve {keys [
         {interpMode=RCIM_Cubic, leaveTangent=16.5652, leaveTangentWeight=0.4385, tangentMode=RCTM_Break, time=0, value=0}
         {arriveTangentWeight=0.0333, interpMode=RCIM_Cubic, leaveTangentWeight=0.9477, tangentMode=RCTM_Break, time=0.1, value=1}
         {arriveTangent=-1.7159, interpMode=RCIM_Cubic, leaveTangent=-1.7159, tangentMode=RCTM_User, time=0.3, value=0.2}
         {arriveTangent=0.2347, interpMode=RCIM_Cubic, leaveTangent=0.2723, tangentMode=RCTM_Break, time=0.6, value=0}
        ]}}}
     Scale Curve=0
    )
    Index 0 Param 4=&FloatFromCurve(FloatCurve=DataInterface() {propertyValues {Curve {keys [{interpMode=RCIM_Cubic, tangentMode=RCTM_Break, time=0.5, value=1}, {interpMode=RCIM_Cubic, tangentMode=RCTM_Break, time=1, value=0}]}}})
    Index 1 Param 1=&FloatFromCurve(
     FloatCurve=DataInterface() {propertyValues {Curve {keys [
         {interpMode=RCIM_Cubic, leaveTangentWeight=0.0833, tangentMode=RCTM_Break, time=0, value=0}
         {arriveTangentWeight=0.0873, interpMode=RCIM_Cubic, tangentMode=RCTM_Break, time=0.25, value=1}
        ]}}}
     Scale Curve=&Lerp_Float(
      A=0.5
      B=0.75
      Alpha=&NormalizeFloat(Value=$Particles.SizeCurve, Min=$Particles.SizeMin, Max=$Particles.SizeMax)
     )
    )
    Index 1 Param 2=&FloatFromCurve(
     FloatCurve=DataInterface() {propertyValues {Curve {keys [
         {interpMode=RCIM_Cubic, tangentMode=RCTM_Break, time=0, value=0.2}
         {interpMode=RCIM_Cubic, leaveTangent=0.9066, tangentMode=RCTM_Break, time=1, value=2}
        ]}}}
     Scale Curve=&RandomRangeFloat(Minimum=0.1, Maximum=0.2)
    )
    Index 1 Param 3=&FloatFromCurve(FloatCurve=DataInterface() {propertyValues {Curve {keys [
         {interpMode=RCIM_Cubic, leaveTangent=6.3917, tangentMode=RCTM_Break, time=0, value=0}
         {arriveTangent=1.7549, interpMode=RCIM_Cubic, leaveTangent=10.6767, tangentMode=RCTM_User, time=0.25, value=1}
        ]}}})
    Index 1 Param 4=&FloatFromCurve(
     FloatCurve=DataInterface() {propertyValues {Curve {keys [{arriveTangent=-0.136, interpMode=RCIM_Cubic, leaveTangent=-0.136, tangentMode=RCTM_User, time=0, value=1}, <at 1 0.95>]}}}
     Scale Curve=&Lerp_Float(
      A=2
      B=5
      Alpha=&NormalizeFloat(Value=$Particles.SizeCurve, Min=$Particles.SizeMin, Max=$Particles.SizeMax)
     )
    )
    Index 2 Param 1=&FloatFromCurve(
     FloatCurve=DataInterface() {propertyValues {Curve {keys [
         {interpMode=RCIM_Cubic, leaveTangent=4.2485, tangentMode=RCTM_Break, time=0, value=0}
         {arriveTangent=1.0548, interpMode=RCIM_Cubic, leaveTangent=1.0548, tangentMode=RCTM_User, time=0.17, value=0.5715}
         <at 1 1>
        ]}}}
     Scale Curve=3
    )
    Index 2 Param 2=&FloatFromCurve(FloatCurve=DataInterface() {propertyValues {Curve {keys [
         {interpMode=RCIM_Cubic, leaveTangentWeight=0.5024, tangentMode=RCTM_User, time=0, value=1}
         {arriveTangentWeight=0.4857, interpMode=RCIM_Cubic, tangentMode=RCTM_User, time=1, value=1}
        ]}}})
    Index 2 Param 3=&FloatFromCurve(
     FloatCurve=DataInterface() {propertyValues {Curve {keys [
         {interpMode=RCIM_Cubic, leaveTangent=-19.1446, leaveTangentWeight=0.1411, tangentMode=RCTM_Break, time=0, value=1}
         {interpMode=RCIM_Cubic, tangentMode=RCTM_Break, time=0.1, value=0}
         {interpMode=RCIM_Cubic, leaveTangent=11.1233, tangentMode=RCTM_Break, time=1, value=1}
        ]}}}
     Scale Curve=5
    )
    Index 2 Param 4=&RandomRangeFloat(Minimum=1, Maximum=1.75)
    Module.Param0WriteEnabled=true
    Module.Param0WriteEnabled1=true
    Module.Param0WriteEnabled2=true
    Module.Param1WriteEnabled=false
    Module.Param1WriteEnabled1=true
    Module.Param1WriteEnabled2=true
    Module.Param2WriteEnabled=true
    Module.Param2WriteEnabled1=true
    Module.Param2WriteEnabled2=true
    Module.Param3WriteEnabled=true
    Module.Param3WriteEnabled1=true
    Module.Param3WriteEnabled2=true
    Write Parameter Index 1=true
    Write Parameter Index 2=true
   )
   /SetVariables(Particles.Sand Dyn Param:Vector4f=&MakeVector4(
     X=&FloatFromCurve(FloatCurve=DataInterface() {propertyValues {Curve {keys [<at 0 -1>,<at 1 1>]}}})
     Y=&FNTimeOfDayScalar(Night Scale=0.7, Morning Scale=0.7, Day Scale=0.1, Evening Scale=0.5)
     Z=&FNTimeOfDayScalar(Night Scale=2, Morning Scale=2, Evening Scale=1.5)
     W=&FNTimeOfDayScalar(Night Scale=0.02, Morning Scale=-0.01, Day Scale=0, Evening Scale=-0.01)
    )) {isSetParameters=true}
  ]
  rendererClasses [&NiagaraSpriteRendererProperties]
  renderers [
   {rendererClass=&NiagaraSpriteRendererProperties}
   {rendererClass=&NiagaraSpriteRendererProperties, rendererIndex=1}
   {rendererClass=&NiagaraSpriteRendererProperties, rendererIndex=2}
  ]
 }
 Streaks {
  emitterUpdateScript [
   &EmitterState()
   &SpawnBurst_Instantaneous(
    Spawn Count=&RandomRangeInt(Minimum=8, Maximum=12)
    Spawn Time=0.105
    Use Spawn Probability=false
    UseLoopCountLimit=false
   ) {module=SpawnBurst_Instantaneous001}
  ]
  particleSpawnScript [
   &InitializeParticle(
    Color=&MakeLinearColorFromVectorAndFloat("Vector (RGB)"=<1 1 1>, "Float (Alpha)"=&RandomRangeFloat(Minimum=0.375, Maximum=0.65))
    Lifetime Max=1.5
    Lifetime Min=0.4
    Lifetime Mode=::NewEnumerator1
    Sprite Size Max=<75 250>
    Sprite Size Min=<35 100>
    Sprite Size Mode=::NewEnumerator2
    UsePositionOffset=false
    Write Material Random=true
   )
   &ShapeLocation(
    Cylinder Height=1
    Cylinder Radius=50
    MaskSpawnGroup=false
    Override Seed=false
    Shape Primitive=::NewEnumerator1
    Surface Only=true
    Use Endcaps In Surface Only Mode=false
    UseOwnerScale=true
   ) {module=ShapeLocation001}
   &AddVelocity(
    Velocity=$Particles.ShapeLocation.ShapeNormal
    Velocity Speed Scale=&MaskFloatBySpawnGroup(Masked Float=&RandomRangeFloat(Minimum=150, Maximum=300), Passthrough Float=0)
   )
   &AddVelocity(Module.Use Falloff Distance=false, Origin Offset=<0 0 -45>, Override Seed=false, Velocity Mode=::NewEnumerator1) {module=AddVelocity001}
   &CameraOffset(Camera Offset Amount=25)
  ]
  particleUpdateScript [
   &ParticleState()
   &ScaleSpriteSize(
    Scale Factor=&Vector2DFromCurve(
     Vector2Curve=DataInterface() {propertyValues {
       XCurve {keys [
         {arriveTangent=0.4813, interpMode=RCIM_Cubic, leaveTangent=0.4813, tangentMode=RCTM_User, time=0, value=0.625}
         {arriveTangent=1.25, interpMode=RCIM_Cubic, leaveTangent=1.25, time=0.4, value=0.9716}
         {arriveTangent=2.6138, interpMode=RCIM_Cubic, leaveTangent=2.6138, tangentMode=RCTM_User, time=0.9, value=1.75}
        ]}
       YCurve {keys [
         {arriveTangent=9.4573, interpMode=RCIM_Cubic, leaveTangent=9.4573, tangentMode=RCTM_User, time=0, value=0}
         {arriveTangent=0.8193, interpMode=RCIM_Cubic, leaveTangent=0.8193, tangentMode=RCTM_User, time=0.2, value=1.5019}
         <at 1 1.7992 RCIM_Cubic>
        ]}
      }}
     ScaleCurve=<1 1>
    )
    Scale Sprite Size Mode=::NewEnumerator2
   )
   &ScaleColor(
    Scale Alpha=&FloatFromCurve(FloatCurve=DataInterface() {propertyValues {Curve {keys [<at 0.37 1 RCIM_Cubic>,<at 1 0 RCIM_Cubic>]}}})
    Scale RGB=&MakeVectorFromLinearColorRGB(Color=&LerpLinearColors(
      StartColor=&RandomRangeLinearColor(
       Minimum=&ColorFromCurve(ColorCurve=DataInterface() {propertyValues {
          AlphaCurve {keys []}
          BlueCurve {keys [<at 0 0.063 RCIM_Cubic>,<at 1 0.1495 RCIM_Cubic>]}
          GreenCurve {keys [<at 0 0.0681 RCIM_Cubic>,<at 1 0.1616 RCIM_Cubic>]}
          RedCurve {keys [<at 0 0.0729 RCIM_Cubic>,<at 1 0.1729 RCIM_Cubic>]}
         }})
       Maximum=&ColorFromCurve(ColorCurve=DataInterface() {propertyValues {
          AlphaCurve {keys []}
          BlueCurve {keys [<at 0 0.1253 RCIM_Cubic>,<at 1 0.2113 RCIM_Cubic>]}
          GreenCurve {keys [<at 0 0.1349 RCIM_Cubic>,<at 1 0.2274 RCIM_Cubic>]}
          RedCurve {keys [<at 0 0.1458 RCIM_Cubic>,<at 1 0.2458 RCIM_Cubic>]}
         }})
      )
      EndColor=&RandomRangeLinearColor(
       Minimum=&ColorFromCurve(ColorCurve=DataInterface() {propertyValues {
          AlphaCurve {keys []}
          BlueCurve {keys [<at 0 0.0828 RCIM_Cubic>,<at 1 0.1153 RCIM_Cubic>]}
          GreenCurve {keys [<at 0 0.1591 RCIM_Cubic>,<at 1 0.2215 RCIM_Cubic>]}
          RedCurve {keys [<at 0 0.2552 RCIM_Cubic>,<at 1 0.3552 RCIM_Cubic>]}
         }})
       Maximum=&ColorFromCurve(ColorCurve=DataInterface() {propertyValues {
          AlphaCurve {keys []}
          BlueCurve {keys [<at 0 0.1952 RCIM_Cubic>,<at 1 0.2504 RCIM_Cubic>]}
          GreenCurve {keys [<at 0 0.253 RCIM_Cubic>,<at 1 0.3244 RCIM_Cubic>]}
          RedCurve {keys [<at 0 0.3542 RCIM_Cubic>,<at 1 0.4542 RCIM_Cubic>]}
         }})
      )
      LerpFactor=&RandomRangeFloat(Minimum=0, Maximum=1)
     ))
    ScaleA=true
    ScaleRGB=true
   )
   &DynamicMaterialParameters(
    Index 0 Param 1=100
    Index 0 Param 4=&FloatFromCurve(FloatCurve=DataInterface() {propertyValues {Curve {keys [{interpMode=RCIM_Cubic, tangentMode=RCTM_User, time=0, value=0}, {interpMode=RCIM_Cubic, tangentMode=RCTM_User, time=0.4, value=1}]}}})
    Index 1 Param 1=&FloatFromCurve(
     FloatCurve=DataInterface() {propertyValues {Curve {keys [
         {interpMode=RCIM_Cubic, tangentMode=RCTM_User, time=0, value=0}
         {arriveTangent=0.6295, interpMode=RCIM_Cubic, leaveTangent=0.6295, tangentMode=RCTM_User, time=1, value=1}
        ]}}}
     Scale Curve=0.42
    )
    Index 1 Param 3=&FloatFromCurve(
     FloatCurve=DataInterface() {propertyValues {Curve {keys [{interpMode=RCIM_Cubic, tangentMode=RCTM_User, time=0.2, value=0.5}, {interpMode=RCIM_Cubic, tangentMode=RCTM_User, time=1, value=1}]}}}
     Scale Curve=2
    )
    Index 1 Param 4=&FloatFromCurve(FloatCurve=DataInterface() {propertyValues {Curve {keys [<at 0 1>,<at 1 0>]}}}, Scale Curve=-0.3)
    Module.Index0Param0Write=true
    Module.Index0Param1Write=false
    Module.Index0Param2Write=false
    Module.Index0Param3Write=true
    Module.Index1Param0Write=true
    Module.Index1Param1Write=true
    Module.Index1Param2Write=true
    Module.Index1Param3Write=true
    Write Parameter Index 1=true
   )
   &Drag(
    Drag=&FloatFromCurve(FloatCurve=DataInterface() {propertyValues {Curve {keys [<at 0.2 1>,<at 0.3 0>]}}}, Scale Curve=15)
    Use Linear Drag=true
    Use Rotational Drag=false
   )
   &SolveForcesAndVelocity(Clamp Velocity=false, Limit Acceleration=false)
  ]
  rendererClasses [&NiagaraSpriteRendererProperties]
  renderers [{rendererClass=&NiagaraSpriteRendererProperties}]
 }
 Streaks_LWE {rendererClasses [], renderers []}
}
summary {
 systemName=NS_QuickBurst_GroundExplosion
 emitters [
  {emitterName=Streaks, rendererClasses [&NiagaraSpriteRendererProperties]}
  {emitterName=Dust, rendererClasses [&NiagaraSpriteRendererProperties]}
  {emitterName=Dirt, simTarget=GPUComputeSim, rendererClasses [&NiagaraSpriteRendererProperties]}
  {emitterName=Dirt_Low, rendererClasses [&NiagaraSpriteRendererProperties]}
  {emitterName=Streaks_LWE, rendererClasses []}
 ]
 userVariables []
}
