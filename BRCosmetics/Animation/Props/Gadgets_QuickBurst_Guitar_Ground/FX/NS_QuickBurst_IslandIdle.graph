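# Niagara idle effect for the island in the QuickBurst Guitar Ground gadget cosmetic, continuously shedding dirt from a disc just below its origin.
# Dirt_Low drops small fading specks, while Dirt_Poofs001 makes wind-pushed dust puffs in two spawn groups, using visibility tags and three sprite renderers.
# Puff materials receive curve-driven dynamic parameters, including time-of-day scaling.
# Debris, sparkle, trailing-sand and original poof emitters are disabled; the LWE emitters list no renderers.
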NiagaraSystem /BRCosmetics/Animation/Props/Gadgets_QuickBurst_Guitar_Ground/FX/NS_QuickBurst_IslandIdle
dependencies {
 usedDataInterfaces [
  &NiagaraDataInterfaceCurve
  &NiagaraDataInterfaceSpriteRendererInfo
  &NiagaraDataInterfaceVector2DCurve
  &NiagaraDataInterfaceColorCurve
 ]
 usedDynamicInputs []
 usedModules [&SystemState]
}
emitters {
 Debris_Elements {
  bEnabled=false
  simTarget=GPUComputeSim
  emitterUpdateScript [
   &EmitterState()
   &SpawnRate(SpawnRate=&RandomRangeFloat(Minimum=0, Maximum=3), Use Spawn Probability=false)
  ]
  particleSpawnScript [
   &InitializeParticle(
    Color Maximum=&RandomRangeLinearColor(Minimum=<rgba 0.5647 0.3607 0.1274 1>, Maximum=<rgba 0.0296 0.02 0.0091 1>)
    Color Minimum=&RandomRangeLinearColor(Minimum=<rgba 0.5647 0.2003 0.1274 1>, Maximum=<rgba 0.0296 0.0125 0.0091 1>)
    Color Mode=::NewEnumerator2
    Lifetime Max=1
    Lifetime Min=0.5
    Lifetime Mode=::NewEnumerator1
    Sprite Rotation Mode=::NewEnumerator2
    Sprite Size Mode=::NewEnumerator1
    Sprite UV Mode=::NewEnumerator3
    Uniform Sprite Size Min=3
    UsePositionOffset=false
    Write Material Random=true
   )
   &ShapeLocation(
    Cylinder Height=1
    Cylinder Radius=50
    MaskSpawnGroup=false
    Offset=<0 0 -20>
    Offset Mode=::NewEnumerator0
    Override Seed=false
    Shape Primitive=::NewEnumerator1
    Surface Only=false
    UseOwnerScale=true
   )
  ]
  particleUpdateScript [
   &ParticleState()
   &ScaleSpriteSize(Uniform Curve Sprite Scale=DataInterface() {propertyValues {Curve {keys [
        {arriveTangentWeight=0.1274, interpMode=RCIM_Cubic, leaveTangentWeight=0.7778, tangentMode=RCTM_User, time=0, value=1}
        {arriveTangent=-0.0405, interpMode=RCIM_Cubic, leaveTangent=-0.0405, tangentMode=RCTM_User, time=1, value=0}
       ]}}})
   &GravityForce()
   &Drag(Drag=&RandomRangeFloat(Minimum=2, Maximum=4), Use Linear Drag=true, Use Rotational Drag=true)
   &Collision(
    Age Colliding Particles=false
    GPU Collision Type=::NewEnumerator2
    KillOnCollision=false
    Module.Control Rest State Via Surface Normal=false
    Module.EnableMaxCollisionCount=false
    Module.Randomize Collision Normal Vector=false
    Restitution=0.3
   )
   &SolveForcesAndVelocity(Clamp Velocity=false, Limit Acceleration=false)
  ]
  rendererClasses [&NiagaraSpriteRendererProperties]
  renderers [{rendererClass=&NiagaraSpriteRendererProperties}]
 }
 Dirt_Low {
  emitterUpdateScript [
   &EmitterState()
   &SpawnRate(Spawn Probability=0.333, SpawnRate=11, Use Spawn Probability=true)
  ]
  particleSpawnScript [
   &InitializeParticle(
    Color=&LerpLinearColors(
     StartColor=&RandomRangeLinearColor(Minimum=<rgba 0.0729 0.0681 0.063 1>, Maximum=<rgba 0.1458 0.1349 0.1253 1>)
     EndColor=&RandomRangeLinearColor(Minimum=<rgba 0.2552 0.1591 0.0828 1>, Maximum=<rgba 0.3542 0.253 0.1952 1>)
     LerpFactor=&RandomRangeFloat(Minimum=0, Maximum=1)
    )
    Lifetime=&RandomRangeFloat(Minimum=0.5, Maximum=1)
    Sprite Rotation Mode=::NewEnumerator2
    Sprite Size Mode=::NewEnumerator1
    Uniform Sprite Size Min=2
    UsePositionOffset=false
    Write Material Random=true
   )
   &ShapeLocation(
    Cylinder Height=1
    Cylinder Radius=66
    MaskSpawnGroup=false
    Offset=<0 0 -22>
    Offset Mode=::NewEnumerator0
    Override Seed=false
    Shape Primitive=::NewEnumerator1
    Surface Only=false
    UseOwnerScale=true
   )
  ]
  particleUpdateScript [
   &ParticleState()
   &GravityForce(Gravity=<0 0 -777>)
   &Drag(Drag=&RandomRangeFloat(Minimum=2, Maximum=4), Use Linear Drag=true, Use Rotational Drag=false)
   &SolveForcesAndVelocity(Clamp Velocity=false, Limit Acceleration=false)
   &ScaleColor(
    Scale Alpha=&FloatFromCurve(FloatCurve=DataInterface() {propertyValues {Curve {keys [{tangentMode=RCTM_User, time=0, value=1}, {tangentMode=RCTM_User, time=1, value=0}]}}})
    Scale RGB=&VectorFromFloat(Value=&MaskFloatBySpawnGroup(Masked Float=&RandomRangeFloat(Minimum=0.1, Maximum=1), Passthrough Float=1))
    ScaleA=true
    ScaleRGB=true
   )
   &ScaleSpriteSize(
    Scale Sprite Size Mode=::NewEnumerator0
    Uniform Scale Factor=&MaskFloatBySpawnGroup(
     Masked Float=&FloatFromCurve(
      FloatCurve=DataInterface() {propertyValues {Curve {keys [
          {arriveTangent=4.1785, interpMode=RCIM_Cubic, leaveTangent=4.1785, tangentMode=RCTM_User, time=0, value=0}
          {arriveTangent=1.1179, interpMode=RCIM_Cubic, leaveTangent=1.1179, tangentMode=RCTM_User, time=0.2042, value=0.6228}
          {arriveTangent=0.0224, interpMode=RCIM_Cubic, leaveTangent=0.0224, tangentMode=RCTM_User, time=1, value=1}
         ]}}}
      Scale Curve=0.85
     )
     Passthrough Float=1
    )
   )
   &SpriteRotationRate(Rotation Rate=&Multiply_Float(
     A=&RandomRangeFloat(Minimum=0.1, Maximum=0.9)
     B=&RandomRangeFloat(Minimum=-90, Maximum=90)
    ))
  ]
  rendererClasses [&NiagaraSpriteRendererProperties]
  renderers [{rendererClass=&NiagaraSpriteRendererProperties}]
 }
 Dirt_Low_LWE {rendererClasses [], renderers []}
 Dirt_Poofs {
  bEnabled=false
  emitterUpdateScript [
   &EmitterState()
   &SpawnRate(
    SpawnRate=&Multiply_Float(A=&RandomRangeFloat(Minimum=5, Maximum=20), B=1)
    Use Spawn Probability=false
   )
  ]
  particleSpawnScript [
   &InitializeParticle(
    Color Maximum=&RandomRangeLinearColor(Minimum=<rgba 0.5647 0.3607 0.1274 0.125>, Maximum=<rgba 0.0296 0.02 0.0091 0.333>)
    Color Minimum=&RandomRangeLinearColor(Minimum=<rgba 0.5647 0.2003 0.1274 0.125>, Maximum=<rgba 0.0296 0.0125 0.0091 0.333>)
    Color Mode=::NewEnumerator2
    Lifetime Max=1
    Lifetime Min=0.5
    Lifetime Mode=::NewEnumerator1
    Sprite Rotation Mode=::NewEnumerator2
    Sprite Size Mode=::NewEnumerator1
    Sprite UV Mode=::NewEnumerator3
    Uniform Sprite Size Max=125
    Uniform Sprite Size Min=50
    UsePositionOffset=false
    Write Material Random=true
   )
   &ShapeLocation(
    Cylinder Height=1
    Cylinder Radius=50
    MaskSpawnGroup=false
    Offset=<0 0 -20>
    Offset Mode=::NewEnumerator0
    Override Seed=false
    Shape Primitive=::NewEnumerator1
    Surface Only=false
    UseOwnerScale=true
   )
   &CameraOffset(Camera Offset Amount=75)
  ]
  particleUpdateScript [
   &ParticleState()
   &ScaleSpriteSize(Uniform Curve Sprite Scale=DataInterface() {propertyValues {Curve {keys [
        {arriveTangent=1.436, interpMode=RCIM_Cubic, leaveTangent=1.436, tangentMode=RCTM_User, time=0, value=0}
        {arriveTangent=1, arriveTangentWeight=0.0664, interpMode=RCIM_Cubic, leaveTangent=1, leaveTangentWeight=0.405, time=0.4, value=0.6}
        {interpMode=RCIM_Cubic, leaveTangent=-0.0405, time=1, value=1}
       ]}}})
   &GravityForce(Gravity=<0 0 -550>)
   &WindForce(Wind Speed=<1 0 0>, Wind Speed Scale=25)
   &Drag(Drag=&RandomRangeFloat(Minimum=8, Maximum=12), Use Linear Drag=true, Use Rotational Drag=true)
   &SolveForcesAndVelocity(Clamp Velocity=false, Limit Acceleration=false)
   &ScaleColor(
    Scale Alpha=&FloatFromCurve(FloatCurve=DataInterface() {propertyValues {Curve {keys [<at 0 1 RCIM_Cubic>,<at 1 0 RCIM_Cubic>]}}})
    ScaleA=true
    ScaleRGB=true
   )
   &SubUVAnimation(
    "Sprite Renderer "=DataInterface() {propertyValues {SpriteRenderer=&self:Dirt_Poofs_6.NiagaraSpriteRendererProperties_0}}
    UseEndFrame=false
    UseStartFrame=false
   )
  ]
  rendererClasses [&NiagaraSpriteRendererProperties]
  renderers [{rendererClass=&NiagaraSpriteRendererProperties}]
 }
 Dirt_Poofs001 {
  emitterUpdateScript [
   &EmitterState()
   &SpawnRate(
    SpawnRate=&Multiply_Float(A=&RandomRangeFloat(Minimum=5, Maximum=10), B=1)
    Use Spawn Probability=false
   )
   &SpawnRate(
    SpawnGroup=2
    SpawnRate=&Multiply_Float(A=&RandomRangeFloat(Minimum=3, Maximum=6), B=1)
    Use Spawn Probability=false
   ) {module=SpawnRate001}
  ]
  particleSpawnScript [
   &InitializeParticle(
    Color=&MakeLinearColorFromVectorAndFloat(
     "Vector (RGB)"=<1 1 1>
     "Float (Alpha)"=&MaskFloatBySpawnGroup(
      Masked Float=&RandomRangeFloat(Minimum=0.125, Maximum=0.25)
      Passthrough Float=&MaskFloatBySpawnGroup(
       Masked Float=&RandomRangeFloat(Minimum=0.25, Maximum=1)
       Passthrough Float=&RandomRangeFloat(Minimum=0.1, Maximum=0.3)
       Spawn Group Mask=1
      )
     )
    )
    Lifetime Max=1.3
    Lifetime Min=0.6
    Lifetime Mode=::NewEnumerator1
    Sprite Rotation Mode=::NewEnumerator2
    Sprite Size Mode=::NewEnumerator1
    Sprite UV Mode=::NewEnumerator3
    Uniform Sprite Size Max=125
    Uniform Sprite Size Min=50
    UsePositionOffset=false
    Write Material Random=true
   )
   &ShapeLocation(
    Cylinder Height=1
    Cylinder Radius=50
    MaskSpawnGroup=false
    Offset=<0 0 -20>
    Offset Mode=::NewEnumerator0
    Override Seed=false
    Shape Primitive=::NewEnumerator1
    Surface Only=false
    UseOwnerScale=true
   )
   &CameraOffset(Camera Offset Amount=22)
   /SetVariables(Particles.VisibilityTag=&MakeCustomIntFromBool(
     Bool=&SetBoolByIntComparison(A=&SpawnGroup(), B=2, Comparison Type=::NewEnumerator2)
     True INT=1
     False INT=0
    )) {isSetParameters=true}
  ]
  particleUpdateScript [
   &ParticleState()
   &DynamicMaterialParameters(
    Index 0 Param 1=&Add_Float(
     A=&RandomRangeFloat(Minimum=0, Maximum=1)
     B=&FloatFromCurve(
      FloatCurve=DataInterface() {propertyValues {Curve {keys [
          {arriveTangent=2.7533, interpMode=RCIM_Cubic, leaveTangent=1.8958, tangentMode=RCTM_User, time=0, value=1}
          {arriveTangent=0.8292, interpMode=RCIM_Cubic, leaveTangent=0.8292, tangentMode=RCTM_User, time=0.321, value=1.3453}
          {arriveTangent=0.0529, interpMode=RCIM_Cubic, leaveTangent=0.2021, tangentMode=RCTM_User, time=1, value=1.5}
         ]}}}
      CurveIndex=&Add_Float(A=&Multiply_Float(A=$Particles.Initial.MaterialRandom, B=0.5), B=$Particles.NormalizedAge)
      Scale Curve=&Lerp_Float(A=0.25, B=0.7, Alpha=&RandomRangeFloat(Minimum=0, Maximum=1))
     )
    )
    Index 0 Param 2=&FloatFromCurve(FloatCurve=DataInterface() {propertyValues {Curve {keys [<at 0 1>,<at 1 0>]}}})
    Index 0 Param 3=&FloatFromCurve(
     FloatCurve=DataInterface() {propertyValues {Curve {keys [
         {interpMode=RCIM_Cubic, leaveTangent=16.5652, leaveTangentWeight=0.4385, tangentMode=RCTM_Break, time=0, value=0}
         {arriveTangentWeight=0.0333, interpMode=RCIM_Cubic, leaveTangentWeight=0.9477, tangentMode=RCTM_Break, time=0.1, value=1}
         {arriveTangent=-1.7159, interpMode=RCIM_Cubic, leaveTangent=-1.7159, tangentMode=RCTM_User, time=0.3, value=0.2}
         {arriveTangent=0.2347, interpMode=RCIM_Cubic, leaveTangent=0.2723, tangentMode=RCTM_Break, time=0.6, value=0}
        ]}}}
     Scale Curve=0
    )
    Index 0 Param 4=&FloatFromCurve(FloatCurve=DataInterface() {propertyValues {Curve {keys [{interpMode=RCIM_Cubic, tangentMode=RCTM_Break, time=0.5, value=1}, {interpMode=RCIM_Cubic, tangentMode=RCTM_Break, time=1, value=0}]}}})
    Index 1 Param 1=&FloatFromCurve(
     FloatCurve=DataInterface() {propertyValues {Curve {keys [
         {interpMode=RCIM_Cubic, leaveTangentWeight=0.0833, tangentMode=RCTM_Break, time=0, value=0}
         {arriveTangentWeight=0.0873, interpMode=RCIM_Cubic, tangentMode=RCTM_Break, time=0.25, value=1}
        ]}}}
     Scale Curve=&Lerp_Float(A=0.5, B=0.75, Alpha=&RandomRangeFloat(Minimum=0, Maximum=1))
    )
    Index 1 Param 2=&FloatFromCurve(
     FloatCurve=DataInterface() {propertyValues {Curve {keys [
         {interpMode=RCIM_Cubic, tangentMode=RCTM_Break, time=0, value=0.2}
         {interpMode=RCIM_Cubic, leaveTangent=0.9066, tangentMode=RCTM_Break, time=1, value=2}
        ]}}}
     Scale Curve=&RandomRangeFloat(Minimum=0.1, Maximum=0.2)
    )
    Index 1 Param 3=&FloatFromCurve(FloatCurve=DataInterface() {propertyValues {Curve {keys [
         {interpMode=RCIM_Cubic, leaveTangent=6.3917, tangentMode=RCTM_Break, time=0, value=0}
         {arriveTangent=1.7549, interpMode=RCIM_Cubic, leaveTangent=10.6767, tangentMode=RCTM_User, time=0.25, value=1}
        ]}}})
    Index 1 Param 4=&FloatFromCurve(
     FloatCurve=DataInterface() {propertyValues {Curve {keys [{arriveTangent=-0.136, interpMode=RCIM_Cubic, leaveTangent=-0.136, tangentMode=RCTM_User, time=0, value=1}, <at 1 0.95>]}}}
     Scale Curve=&Lerp_Float(A=2, B=5, Alpha=&RandomRangeFloat(Minimum=0, Maximum=1))
    )
    Index 2 Param 1=&FloatFromCurve(
     FloatCurve=DataInterface() {propertyValues {Curve {keys [
         {interpMode=RCIM_Cubic, leaveTangent=4.2485, tangentMode=RCTM_Break, time=0, value=0}
         {arriveTangent=1.0548, interpMode=RCIM_Cubic, leaveTangent=1.0548, tangentMode=RCTM_User, time=0.17, value=0.5715}
         <at 1 1>
        ]}}}
     Scale Curve=3
    )
    Index 2 Param 2=&FloatFromCurve(FloatCurve=DataInterface() {propertyValues {Curve {keys [
         {interpMode=RCIM_Cubic, leaveTangentWeight=0.5024, tangentMode=RCTM_User, time=0, value=1}
         {arriveTangentWeight=0.4857, interpMode=RCIM_Cubic, tangentMode=RCTM_User, time=1, value=1}
        ]}}})
    Index 2 Param 3=&FloatFromCurve(
     FloatCurve=DataInterface() {propertyValues {Curve {keys [
         {interpMode=RCIM_Cubic, leaveTangent=-19.1446, leaveTangentWeight=0.1411, tangentMode=RCTM_Break, time=0, value=1}
         {interpMode=RCIM_Cubic, tangentMode=RCTM_Break, time=0.1, value=0}
         {interpMode=RCIM_Cubic, leaveTangent=11.1233, tangentMode=RCTM_Break, time=1, value=1}
        ]}}}
     Scale Curve=5
    )
    Index 2 Param 4=&RandomRangeFloat(Minimum=1, Maximum=1.75)
    Module.Param0WriteEnabled=true
    Module.Param0WriteEnabled1=true
    Module.Param0WriteEnabled2=true
    Module.Param1WriteEnabled=false
    Module.Param1WriteEnabled1=true
    Module.Param1WriteEnabled2=true
    Module.Param2WriteEnabled=true
    Module.Param2WriteEnabled1=true
    Module.Param2WriteEnabled2=true
    Module.Param3WriteEnabled=true
    Module.Param3WriteEnabled1=true
    Module.Param3WriteEnabled2=true
    Write Parameter Index 1=true
    Write Parameter Index 2=true
   )
   &ScaleSpriteSize(Uniform Curve Sprite Scale=DataInterface() {propertyValues {Curve {keys [
        {arriveTangent=1.436, interpMode=RCIM_Cubic, leaveTangent=1.436, tangentMode=RCTM_User, time=0, value=0.333}
        {interpMode=RCIM_Cubic, leaveTangent=-0.0405, time=1, value=1}
       ]}}})
   &GravityForce(Gravity=<0 0 -550>)
   &WindForce(Wind Speed=<1 0 0>, Wind Speed Scale=25)
   &Drag(Drag=&RandomRangeFloat(Minimum=8, Maximum=12), Use Linear Drag=true, Use Rotational Drag=true)
   &SolveForcesAndVelocity(Clamp Velocity=false, Limit Acceleration=false)
   &ScaleColor(
    Scale Alpha=&FloatFromCurve(FloatCurve=DataInterface() {propertyValues {Curve {keys [
         {interpMode=RCIM_Cubic, tangentMode=RCTM_User, time=0, value=1}
         {arriveTangent=-2.0933, interpMode=RCIM_Cubic, tangentMode=RCTM_Break, time=1, value=0}
        ]}}})
    Scale RGB=&MakeVectorFromLinearColorRGB(Color=&LerpLinearColors(
      StartColor=&RandomRangeLinearColor(
       Minimum=&ColorFromCurve(ColorCurve=DataInterface() {propertyValues {
          AlphaCurve {keys []}
          BlueCurve {keys [<at 0 0.063 RCIM_Cubic>,<at 1 0.1495 RCIM_Cubic>]}
          GreenCurve {keys [<at 0 0.0681 RCIM_Cubic>,<at 1 0.1616 RCIM_Cubic>]}
          RedCurve {keys [<at 0 0.0729 RCIM_Cubic>,<at 1 0.1729 RCIM_Cubic>]}
         }})
       Maximum=&ColorFromCurve(ColorCurve=DataInterface() {propertyValues {
          AlphaCurve {keys []}
          BlueCurve {keys [<at 0 0.1253 RCIM_Cubic>,<at 1 0.2113 RCIM_Cubic>]}
          GreenCurve {keys [<at 0 0.1349 RCIM_Cubic>,<at 1 0.2274 RCIM_Cubic>]}
          RedCurve {keys [<at 0 0.1458 RCIM_Cubic>,<at 1 0.2458 RCIM_Cubic>]}
         }})
      )
      EndColor=&RandomRangeLinearColor(
       Minimum=&ColorFromCurve(ColorCurve=DataInterface() {propertyValues {
          AlphaCurve {keys []}
          BlueCurve {keys [<at 0 0.0828 RCIM_Cubic>,<at 1 0.1153 RCIM_Cubic>]}
          GreenCurve {keys [<at 0 0.1591 RCIM_Cubic>,<at 1 0.2215 RCIM_Cubic>]}
          RedCurve {keys [<at 0 0.2552 RCIM_Cubic>,<at 1 0.3552 RCIM_Cubic>]}
         }})
       Maximum=&ColorFromCurve(ColorCurve=DataInterface() {propertyValues {
          AlphaCurve {keys []}
          BlueCurve {keys [<at 0 0.1952 RCIM_Cubic>,<at 1 0.2504 RCIM_Cubic>]}
          GreenCurve {keys [<at 0 0.253 RCIM_Cubic>,<at 1 0.3244 RCIM_Cubic>]}
          RedCurve {keys [<at 0 0.3542 RCIM_Cubic>,<at 1 0.4542 RCIM_Cubic>]}
         }})
      )
      LerpFactor=&RandomRangeFloat(Minimum=0, Maximum=1)
     ))
    ScaleA=true
    ScaleRGB=true
   )
   /SetVariables(Particles.Sand Dyn Param:Vector4f=&MakeVector4(
     X=&FloatFromCurve(FloatCurve=DataInterface() {propertyValues {Curve {keys [<at 0 -1>,<at 1 1>]}}})
     Y=&FNTimeOfDayScalar(Night Scale=0.7, Morning Scale=0.7, Day Scale=0.1, Evening Scale=0.5)
     Z=&FNTimeOfDayScalar(Night Scale=2, Morning Scale=2, Evening Scale=1.5)
     W=&FNTimeOfDayScalar(Night Scale=0.02, Morning Scale=-0.01, Day Scale=0, Evening Scale=-0.01)
    )) {isSetParameters=true}
  ]
  rendererClasses [&NiagaraSpriteRendererProperties]
  renderers [
   {rendererClass=&NiagaraSpriteRendererProperties}
   {rendererClass=&NiagaraSpriteRendererProperties, rendererIndex=1}
   {rendererClass=&NiagaraSpriteRendererProperties, rendererIndex=2}
  ]
 }
 Dirt_Poofs_LWE {rendererClasses [], renderers []}
 Sand_Sparkles {
  bEnabled=false
  emitterUpdateScript [
   &EmitterState()
   &SpawnRate(
    SpawnRate=&Multiply_Float(A=&RandomRangeFloat(Minimum=0, Maximum=10), B=1)
    Use Spawn Probability=false
   )
  ]
  particleSpawnScript [
   &InitializeParticle(
    Color Maximum=&RandomRangeLinearColor(Minimum=<rgba 0.5647 0.3607 0.1274 0.05>, Maximum=<rgba 0.0296 0.02 0.0091 0.125>)
    Color Minimum=&RandomRangeLinearColor(Minimum=<rgba 0.5647 0.2878 0.1274 0.05>, Maximum=<rgba 0.0296 0.0166 0.0091 0.125>)
    Color Mode=::NewEnumerator2
    Lifetime Max=1
    Lifetime Min=0.5
    Lifetime Mode=::NewEnumerator1
    Sprite Rotation Mode=::NewEnumerator2
    Sprite Size Mode=::NewEnumerator1
    Sprite UV Mode=::NewEnumerator3
    Uniform Sprite Size Max=100
    Uniform Sprite Size Min=25
    UsePositionOffset=false
    Write Material Random=true
   )
   &ShapeLocation(
    Cylinder Height=1
    Cylinder Radius=50
    MaskSpawnGroup=false
    Offset=<0 0 -20>
    Offset Mode=::NewEnumerator0
    Override Seed=false
    Shape Primitive=::NewEnumerator1
    Surface Only=false
    UseOwnerScale=true
   )
   &CameraOffset(Camera Offset Amount=50)
  ]
  particleUpdateScript [
   &ParticleState()
   &ScaleSpriteSize(Uniform Curve Sprite Scale=DataInterface() {propertyValues {Curve {keys [
        {arriveTangent=1.436, interpMode=RCIM_Cubic, leaveTangent=1.436, tangentMode=RCTM_User, time=0, value=0}
        {arriveTangent=1, arriveTangentWeight=0.0664, interpMode=RCIM_Cubic, leaveTangent=1, leaveTangentWeight=0.405, time=0.4, value=0.6}
        {interpMode=RCIM_Cubic, leaveTangent=-0.0405, time=1, value=1}
       ]}}})
   &GravityForce(Gravity=<0 0 -550>)
   &WindForce(Wind Speed=<1 0 0>, Wind Speed Scale=25)
   &Drag(Drag=&RandomRangeFloat(Minimum=8, Maximum=12), Use Linear Drag=true, Use Rotational Drag=true)
   &SolveForcesAndVelocity(Clamp Velocity=false, Limit Acceleration=false)
   &ScaleColor(
    Scale Alpha=&FloatFromCurve(FloatCurve=DataInterface() {propertyValues {Curve {keys [<at 0 1 RCIM_Cubic>,<at 1 0 RCIM_Cubic>]}}})
    ScaleA=true
    ScaleRGB=true
   )
   &SubUVAnimation(
    Evaluation Type=::NewEnumerator0
    "Sprite Renderer "=DataInterface() {propertyValues {SpriteRenderer=&self:Sand_Sparkles_5.NiagaraSpriteRendererProperties_0}}
    SubUV Animation Mode:ENiagara_SubUVLookupModeV2=::NewEnumerator2
    UseEndFrame=false
    UseStartFrame=false
   )
  ]
  rendererClasses [&NiagaraSpriteRendererProperties]
  renderers [{rendererClass=&NiagaraSpriteRendererProperties}]
 }
 Trailing_Sand {
  bEnabled=false
  emitterUpdateScript [
   &EmitterState()
   &SpawnRate(SpawnRate=&RandomRangeFloat(Minimum=333, Maximum=333), Use Spawn Probability=false)
  ]
  particleSpawnScript [
   &InitializeParticle(
    Color Maximum=&RandomRangeLinearColor(Minimum=<rgba 0.5647 0.3607 0.1274 1>, Maximum=<rgba 0.5647 0.2003 0.1274 1>)
    Color Minimum=&RandomRangeLinearColor(Minimum=<rgba 0.5647 0.3607 0.1274 1>, Maximum=<rgba 0.5647 0.2003 0.1274 1>)
    Color Mode=::NewEnumerator2
    Lifetime Max=0.6
    Lifetime Min=0.25
    Lifetime Mode=::NewEnumerator1
    Sprite Size Max=<10 15>
    Sprite Size Min=<5 10>
    Sprite Size Mode=::NewEnumerator2
    UsePositionOffset=true
    Write Material Random=true
   )
   &ShapeLocation(
    Cylinder Height=1
    Cylinder Radius=50
    MaskSpawnGroup=false
    Offset=<0 0 -20>
    Offset Mode=::NewEnumerator0
    Override Seed=false
    Shape Primitive=::NewEnumerator1
    Surface Only=false
    UseOwnerScale=true
   )
   &CameraOffset(Camera Offset Amount=&RandomRangeFloat(Minimum=-25, Maximum=25))
  ]
  particleUpdateScript [
   &ParticleState()
   &GravityForce(Gravity=&Multiply_VectorByFloat(Vector=<0 0 -700>, Float=1))
   &InheritVelocity(Inherited Velocity Amount Scale=<-0.05 -0.05 -0.05>, Inherited Velocity Speed Limit=25, Speed Limit=true, UseSpeedThreshold=false)
   &SolveForcesAndVelocity(Clamp Velocity=false, Limit Acceleration=false)
   &ScaleSpriteSize(
    Non-Uniform Curve Sprite Scale=DataInterface() {propertyValues {
      XCurve {keys [{tangentMode=RCTM_User, time=0, value=1}, {tangentMode=RCTM_User, time=1, value=1}]}
      YCurve {keys [{arriveTangent=3.1374, interpMode=RCIM_Cubic, leaveTangent=3.1374, tangentMode=RCTM_User, time=0, value=0.25}, <at 1 3 RCIM_Cubic>]}
     }}
    Scale Sprite Size Mode=::NewEnumerator3
   )
   &ScaleSpriteSizeBySpeed(Max Scale Factor=<1 4>, Min Scale Factor=<0.5 1>)
   &ScaleColor(
    Scale Alpha=&FloatFromCurve(FloatCurve=DataInterface() {propertyValues {Curve {keys [<at 0 1 RCIM_Cubic>,<at 1 0 RCIM_Cubic>]}}}, Scale Curve=2)
    ScaleA=true
    ScaleRGB=true
   )
  ]
  rendererClasses [&NiagaraSpriteRendererProperties]
  renderers [{rendererClass=&NiagaraSpriteRendererProperties}]
 }
}
summary {
 systemName=NS_QuickBurst_IslandIdle
 emitters [
  {bEnabled=false, emitterName=Dirt_Poofs, rendererClasses [&NiagaraSpriteRendererProperties]}
  {bEnabled=false, emitterName=Debris_Elements, simTarget=GPUComputeSim, rendererClasses [&NiagaraSpriteRendererProperties]}
  {bEnabled=false, emitterName=Sand_Sparkles, rendererClasses [&NiagaraSpriteRendererProperties]}
  {bEnabled=false, emitterName=Trailing_Sand, rendererClasses [&NiagaraSpriteRendererProperties]}
  {emitterName=Dirt_Poofs001, rendererClasses [&NiagaraSpriteRendererProperties]}
  {emitterName=Dirt_Low, rendererClasses [&NiagaraSpriteRendererProperties]}
  {emitterName=Dirt_Poofs_LWE, rendererClasses []}
  {emitterName=Dirt_Low_LWE, rendererClasses []}
 ]
 userVariables []
}
