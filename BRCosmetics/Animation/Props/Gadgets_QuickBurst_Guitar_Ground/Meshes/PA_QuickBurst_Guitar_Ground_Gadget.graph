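# Physics asset for the ground piece of the QuickBurst Guitar gadget prop among the BR cosmetics, giving its skeleton simple collision bodies.
# Each of the four bones has one capsule: a very large, near-spherical one on root, tiny ones on offset and main, and a sizeable off-centre one on JNT_Ground.
# Three constraints chain root, offset, main and JNT_Ground in order, all sharing identical swing and twist limit angles.
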PhysicsAsset /BRCosmetics/Animation/Props/Gadgets_QuickBurst_Guitar_Ground/Meshes/PA_QuickBurst_Guitar_Ground_Gadget
bodies {
 JNT_Ground [{shapeType=Capsule, length=92.4026, radius=85.4919, center=<0.0235 0.4408 -22.3769>, rotation=<rot 0 -80.378 -86.427>}]
 main [{shapeType=Capsule, length=0.505, radius=0.505, rotation=<rot 0 -90 0>}]
 offset [{shapeType=Capsule, length=0.505, radius=0.505, rotation=<rot 0 -90 0>}]
 root [{shapeType=Capsule, length=0.5049, radius=249.0217, rotation=<rot 0 -90 0>}]
}
constraints [
 {bone1Name=offset, bone2Name=root, swing1LimitDegrees=45, swing2LimitDegrees=45, twistLimitDegrees=45}
 {bone1Name=main, bone2Name=offset, swing1LimitDegrees=45, swing2LimitDegrees=45, twistLimitDegrees=45}
 {bone1Name=JNT_Ground, bone2Name=main, swing1LimitDegrees=45, swing2LimitDegrees=45, twistLimitDegrees=45}
]
